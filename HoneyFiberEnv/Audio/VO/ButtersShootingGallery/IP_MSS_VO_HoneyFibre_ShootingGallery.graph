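# Looping mono MetaSound for shooting gallery NPC VO in HoneyFiberEnv, with an ambient "PewPewPew" line and reactions to front enter and exit triggers.
# Ambient repeats at a random interval; enter or exit halts it with an ADSR fade.
# Enter picks from a wave array and exit plays "Hamburgers", both gated against overlap until a random post-line cooldown.
# Deferred trigger variables feed back to reset gates and timer; a bus send feeds music sidechaining.
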MetaSoundSource /HoneyFiberEnv/Audio/VO/ButtersShootingGallery/IP_MSS_VO_HoneyFibre_ShootingGallery
inputs [
 Play.Front.Enter {type=Trigger}
 Play.Front.Exit {type=Trigger}
 UE.Source.OnPlay {type=Trigger}
]
interfaces [UE.OutputFormat.Mono:1.0, UE.Source:1.1]
outputs ["UE.OutputFormat.Mono.Audio:0" {type=Audio}]
pages [Default {
  input_defaults {Play.Front.Enter=null, Play.Front.Exit=null, UE.Source.OnPlay=null}
  nodes [
   TriggerAnyTriggerAny2_1 = "TriggerAny.Trigger Any (2)"(In 0=^Play.Front.Enter, In 1=^Play.Front.Exit)
   n1 = VariableDeferredAccessor.Trigger(Variable=InitVariable.Trigger().Variable)
   n2 = VariableDeferredAccessor.Trigger(Variable=@n1.Variable)
   n3 = VariableDeferredAccessor.Trigger(Variable=InitVariable.Trigger().Variable)
   n4 = VariableDeferredAccessor.Trigger(Variable=@n3.Variable)
   TriggerAnyTriggerAny2_3 = "TriggerAny.Trigger Any (2)"(In 1=@n2.Value, In 0=@n4.Value)
   TriggerAnyTriggerAny2_4 = "TriggerAny.Trigger Any (2)"(In 0=^Play.Front.Enter, In 1=^Play.Front.Exit)
   TriggerAnyTriggerAny3_0 = "TriggerAny.Trigger Any (3)"(In 2=^UE.Source.OnPlay, In 1=@n2.Value, In 0=@n4.Value)
   n5 = UE.RandomTime(Next=@TriggerAnyTriggerAny3_0.Out) {also {defaults {Max=18, Min=12}}}
   TriggerControl_2 = UE.Trigger Control(
    Open=@TriggerAnyTriggerAny2_3.Out
    Close=@TriggerAnyTriggerAny2_4.Out
    Trigger In=UE.TriggerRepeat(Start=@n5.On Next, Period=@n5.Value, Stop=@TriggerAnyTriggerAny2_4.Out).RepeatOut
   )
   n6 = VariableDeferredAccessor.Trigger(Variable=InitVariable.Trigger().Variable)
   WavePlayer_0 = UE.Wave Player:Mono(Play=@TriggerControl_2.Trigger Out, Stop=@n6.Value) {also {defaults {Wave Asset=/HoneyFiberEnv/Audio/VO/ButtersShootingGallery/Waves/IP_VO_HoneyFiber_ShootingGallery_Ambient_PewPewPew}, version [1,1]}}
   ADSREnvelopeADSREnvelope_0 = ADSR Envelope.ADSR Envelope:Float(Trigger Release=@TriggerAnyTriggerAny2_1.Out, Trigger Attack=@WavePlayer_0.On Play) {also {defaults {Release Time=0.1, Sustain Level=1}}}
   n7 = VariableDeferredAccessor.Trigger(Variable=InitVariable.Trigger().Variable)
   n8 = VariableDeferredAccessor.Trigger(Variable=InitVariable.Trigger().Variable)
   TriggerAnyTriggerAny2_0 = "TriggerAny.Trigger Any (2)"(In 1=@n7.Value, In 0=@n8.Value)
   TriggerAnyTriggerAny2_2 = "TriggerAny.Trigger Any (2)"(In 1=@n3.Value, In 0=@n1.Value)
   TriggerControl_1 = UE.Trigger Control(Trigger In=^Play.Front.Exit, Close=@TriggerAnyTriggerAny2_0.Out, Open=@TriggerAnyTriggerAny2_2.Out)
   WavePlayer_1 = UE.Wave Player:Mono(Play=@TriggerControl_1.Trigger Out) {also {defaults {Wave Asset=/HoneyFiberEnv/Audio/VO/ButtersShootingGallery/Waves/IP_VO_HoneyFiber_ShootingGallery_OnExit_Hamburgers}, version [1,1]}}
   TriggerControl_0 = UE.Trigger Control(Trigger In=^Play.Front.Enter, Close=@TriggerAnyTriggerAny2_0.Out, Open=@TriggerAnyTriggerAny2_2.Out)
   ArrayRandomGet_0 = Array.Random Get:WaveAsset:Array(Next=@TriggerControl_0.Trigger Out, In Array=VariableAccessor.WaveAsset:Array(Variable=InitVariable.WaveAsset:Array().Variable).Value) {also {defaults {Weights []}}}
   WavePlayer_2 = UE.Wave Player:Mono(Play=@ArrayRandomGet_0.On Next, Wave Asset=@ArrayRandomGet_0.Value) {also {version [1,1]}}
   AudioMixerAudioMixerMono3_0 = "AudioMixer.Audio Mixer (Mono, 3)"(
    In 0=UE.Reroute(Value=UE.Multiply:Audio by Float(AdditionalOperands=@ADSREnvelopeADSREnvelope_0.Out Envelope, PrimaryOperand=@WavePlayer_0.Out Mono).Out).Value
    In 2=@WavePlayer_1.Out Mono
    In 1=@WavePlayer_2.Out Mono
   ) {also {defaults {Gain 0=0.75}}}
   TriggerAnyTriggerAny2_5 = "TriggerAny.Trigger Any (2)"(In 1=@WavePlayer_1.On Finished, In 0=@WavePlayer_2.On Finished)
   n9 = UE.RandomTime(Next=@TriggerAnyTriggerAny2_5.Out) {also {defaults {Max=5, Min=3}}}
   TriggerDelay_0 = UE.Trigger Delay(In=@n9.On Next, Delay Time=@n9.Value)
   VariableMutator.Trigger(Value=@TriggerDelay_0.Out, Variable=@n2.Variable)
   VariableMutator.Trigger(Value=@ADSREnvelopeADSREnvelope_0.On Done, Variable=@n6.Variable)
   VariableMutator.Trigger(Variable=@n8.Variable, Value=@WavePlayer_2.On Play)
   VariableMutator.Trigger(Variable=@n7.Variable, Value=@WavePlayer_1.On Play)
  ]
  links [{dst="@UE.OutputFormat.Mono.Audio:0", src=AudioMixerAudioMixerMono3_0.Out}]
 }]
properties {
 assetImportData=&self:AssetImportData
 attenuationSettings=&/IP_ATT_HoneyFibre_ShootingGallery
 bDebug=true
 bLooping=true
 duration=10000
 maxDistance=2999.2888
 sampleRate=48000
 soundClassObject=&Fort_VO_NPC
 busSends [{audioBus=&AB_Sidechain_FortMusic_Vol, sendLevel=1}]
}
